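; Bug hut set-piece actor for the BigSplat set, with optional smoke and bugs.
; The construction script runs the parent's, then uses ShowSmoke to toggle visibility of the wispy smoke splines component.
; ShowBugs toggles the root component of the child actor, after a validity check, with both toggles propagating to child components.
; BeginPlay, overlap and Tick are disabled stubs, so the Blueprint adds no runtime logic.
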(event EventBeginPlay
  (|CallParentFunction) ; This node is disabled and will not be called.\nDrag off pins to build functionality.
)

(event Collision|EventActorBeginOverlap
  (params (OtherActor :type Actor%20Object%20Reference))
  (|CallParentFunction OtherActor) ; This node is disabled and will not be called.\nDrag off pins to build functionality.
)

(event EventTick
  (params (DeltaSeconds :type Float%20%28single-precision%29))
  (|CallParentFunction DeltaSeconds) ; This node is disabled and will not be called.\nDrag off pins to build functionality.
)

(fn ConstructionScript
  (bind _childactor (Variables|Default|GetChildActor))
  (bind _childactor_2 (Variables|/Script/Engine.ChildActorComponent|GetChildActor _childactor))
  (bind _childactor_3 (Variables|Default|GetChildActor))
  (bind _childactor_4 (Variables|/Script/Engine.ChildActorComponent|GetChildActor _childactor_3))
  (|CallParentFunction)
  (Utilities|FlowControl|Sequence
    (:then_0
      (if (Variables|Default|GetShowSmoke)
        (Rendering|SetVisibility (Variables|Default|GetSM_BugHut_Smoke_Wispy_Splines) true true)
        (else
          (Rendering|SetVisibility (Variables|Default|GetSM_BugHut_Smoke_Wispy_Splines) false true))))
    (:then_1
      (if (Variables|Default|GetShowBugs)
        (Utilities|IsValid _childactor_4
          (:Is%20Valid
            (Rendering|SetVisibility (Variables|/Script/Engine.Actor|GetRootComponent _childactor_4) true true)))
        (else
          (Utilities|IsValid _childactor_2
            (:Is%20Valid
              (Rendering|SetVisibility (Variables|/Script/Engine.Actor|GetRootComponent _childactor_2) false true))))))))
